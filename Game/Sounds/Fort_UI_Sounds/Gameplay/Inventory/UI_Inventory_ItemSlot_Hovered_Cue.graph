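# Sound cue for the UI hover sound played over inventory item slots.
# A single wave player feeds the output directly, reusing the front-end default hover wave at reduced volume.
# It uses the non-critical in-game UI sound class with zero priority and virtualization disabled.
# A concurrency override allows only one instance, stopping the oldest so rapid hovers do not stack.
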SoundCue /Game/Sounds/Fort_UI_Sounds/Gameplay/Inventory/UI_Inventory_ItemSlot_Hovered_Cue
root=WavePlayer_3
nodes [{id=WavePlayer_3, properties {soundWaveAssetPtr=/Game/Sounds/Fort_UI_Sounds/FrontEnd/DefaultSounds/PressedButtons/wav/UI_MP_Hover_Default_01}}]
properties {
 bOverrideConcurrency=true
 maxDistance=2097152
 priority=0
 soundClassObject=&Fort_UI_InGame_NonCritical
 virtualizationMode=Disabled
 volumeMultiplier=0.6
 concurrencyOverrides {maxCount=1, resolutionRule=StopOldest}
}
